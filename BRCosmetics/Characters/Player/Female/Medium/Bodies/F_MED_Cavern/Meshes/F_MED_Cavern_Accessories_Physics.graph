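# Physics asset for the accessories of the medium female Cavern player body cosmetic, covering shirt pieces, a necklace, a pouch tab and a zipper.
# Capsules and boxes on pelvis, spine and left thigh form the body volumes, chained by free-swing, free-twist constraints.
# Four shirt bodies hang directly from spine_05 with tight angular limits.
# Necklace and pouch pieces use asymmetric limits against tiny MSTR anchor bodies, themselves free-jointed to spine_05 or thigh_l.
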PhysicsAsset /BRCosmetics/Characters/Player/Female/Medium/Bodies/F_MED_Cavern/Meshes/F_MED_Cavern_Accessories_Physics
bodies {
 dyn_back_shirt [{shapeType=Box, extentX=10.6353, extentY=18.9294, extentZ=1.7159, center=<5.134 0 0.3357>}]
 dyn_front_shirt [{shapeType=Capsule, length=12.3091, radius=4.0699, center=<4.0549 0.0001 -0.9107>, rotation=<rot 0 0 90>}]
 dyn_left_shirt [{shapeType=Capsule, length=8.5972, radius=1.7098, center=<2.8369 0.0001 0>, rotation=<rot 0 0 90>}]
 dyn_necklace [{shapeType=Box, extentX=6.5796, extentY=7.0569, extentZ=1.3476, center=<3.0385 0 0>}]
 dyn_necklace_MSTR [{shapeType=Capsule, length=0.505, radius=0.505}]
 dyn_pouchTab [{shapeType=Box, extentX=2.7866, extentY=0.9258, extentZ=3.7425, center=<-1.3795 -0.0025 0.0006>, rotation=<rot 0 -90.9331 -88.7911>}]
 dyn_pouchTab_MSTR [{shapeType=Capsule, length=0.505, radius=0.505}]
 dyn_pouchZipper [{shapeType=Box, extentX=4.1537, extentY=2.0924, extentZ=1.01, center=<-2.0197 0 0>}]
 dyn_pouchZipper_MSTR [{shapeType=Capsule, length=0.505, radius=0.505, rotation=<rot 0 -90 0>}]
 dyn_right_shirt [{shapeType=Capsule, length=8.9606, radius=1.7098, center=<2.8369 0.0001 0>, rotation=<rot 0 0 90>}]
 pelvis [{shapeType=Capsule, length=12.4714, radius=9.3267}]
 spine_02 [{shapeType=Capsule, length=10.8036, radius=7.6499}]
 spine_03 [{shapeType=Capsule, length=4.7209, radius=7.1717}]
 spine_04 [{shapeType=Capsule, length=11.5901, radius=7.7896, center=<3.7757 3.779 0>}, {shapeType=Capsule, length=3.7495, radius=6.0669}]
 spine_05 [{shapeType=Capsule, length=16.1126, radius=9.5448}]
 thigh_l [{shapeType=Capsule, length=25.2148, radius=8.2607, center=<-21.3131 0 0>, rotation=<rot 90 0 0>}]
 thigh_twist_01_l [{shapeType=Box, extentX=17.393, extentY=10.1103, extentZ=7.7743, center=<9.8762 1.6675 -7.5312>, rotation=<rot -5.0593 0 -4.9147>}]
}
constraints [
 {
  bone1Name=thigh_twist_01_l
  bone2Name=thigh_l
  swing1LimitDegrees=45
  swing1Motion=Free
  swing2LimitDegrees=45
  swing2Motion=Free
  twistLimitDegrees=45
  twistMotion=Free
 }
 {
  bone1Name=thigh_l
  bone2Name=pelvis
  swing1LimitDegrees=45
  swing1Motion=Free
  swing2LimitDegrees=45
  swing2Motion=Free
  twistLimitDegrees=45
  twistMotion=Free
 }
 {
  bone1Name=dyn_pouchZipper_MSTR
  bone2Name=thigh_l
  swing1LimitDegrees=45
  swing1Motion=Free
  swing2LimitDegrees=45
  swing2Motion=Free
  twistLimitDegrees=45
  twistMotion=Free
 }
 {bone1Name=dyn_pouchZipper, bone2Name=dyn_pouchZipper_MSTR, swing1LimitDegrees=5, swing2LimitDegrees=120, twistLimitDegrees=10}
 {
  bone1Name=dyn_pouchTab_MSTR
  bone2Name=thigh_l
  swing1LimitDegrees=45
  swing1Motion=Free
  swing2LimitDegrees=45
  swing2Motion=Free
  twistLimitDegrees=45
  twistMotion=Free
 }
 {bone1Name=dyn_pouchTab, bone2Name=dyn_pouchTab_MSTR, swing1LimitDegrees=20, swing2LimitDegrees=45, twistLimitDegrees=20}
 {
  bone1Name=spine_05
  bone2Name=spine_04
  swing1LimitDegrees=45
  swing1Motion=Free
  swing2LimitDegrees=45
  swing2Motion=Free
  twistLimitDegrees=45
  twistMotion=Free
 }
 {
  bone1Name=spine_04
  bone2Name=spine_03
  swing1LimitDegrees=45
  swing1Motion=Free
  swing2LimitDegrees=45
  swing2Motion=Free
  twistLimitDegrees=45
  twistMotion=Free
 }
 {bone1Name=dyn_necklace, bone2Name=dyn_necklace_MSTR, swing1LimitDegrees=20, swing2LimitDegrees=110, twistLimitDegrees=10}
 {
  bone1Name=dyn_necklace_MSTR
  bone2Name=spine_05
  swing1LimitDegrees=45
  swing1Motion=Free
  swing2LimitDegrees=45
  swing2Motion=Free
  twistLimitDegrees=45
  twistMotion=Free
 }
 {bone1Name=dyn_left_shirt, bone2Name=spine_05, swing1LimitDegrees=25, swing2LimitDegrees=25, twistLimitDegrees=5}
 {bone1Name=dyn_front_shirt, bone2Name=spine_05, swing1LimitDegrees=25, swing2LimitDegrees=25, twistLimitDegrees=5}
 {bone1Name=dyn_back_shirt, bone2Name=spine_05, swing1LimitDegrees=25, swing2LimitDegrees=25, twistLimitDegrees=5}
 {bone1Name=dyn_right_shirt, bone2Name=spine_05, swing1LimitDegrees=25, swing2LimitDegrees=25, twistLimitDegrees=5}
 {
  bone1Name=spine_03
  bone2Name=spine_02
  swing1LimitDegrees=45
  swing1Motion=Free
  swing2LimitDegrees=45
  swing2Motion=Free
  twistLimitDegrees=45
  twistMotion=Free
 }
 {
  bone1Name=spine_02
  bone2Name=pelvis
  swing1LimitDegrees=45
  swing1Motion=Free
  swing2LimitDegrees=45
  swing2Motion=Free
  twistLimitDegrees=45
  twistMotion=Free
 }
]
