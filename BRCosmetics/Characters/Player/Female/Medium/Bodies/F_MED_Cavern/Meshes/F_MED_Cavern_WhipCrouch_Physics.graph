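# Physics asset for the whip and dangling accessories on the F_MED_Cavern female character, named as the crouch variant.
# A seven-segment capsule whip chain hangs from a master body on the pelvis, with swing and twist limits between segments.
# The hilt, boxed whip base and two rope pieces attach separately to pelvis or spine with angular limits.
# Pelvis, spine, both thighs and only the right calf supply body volumes, joined by free angular constraints.
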PhysicsAsset /BRCosmetics/Characters/Player/Female/Medium/Bodies/F_MED_Cavern/Meshes/F_MED_Cavern_WhipCrouch_Physics
bodies {
 calf_r [{shapeType=Capsule, length=22.4772, radius=6.3509, center=<15.5917 2.1298 -0.0006>, rotation=<rot 90 172.875 166.2337>}]
 dyn_hilt [{shapeType=Capsule, length=7.7099, radius=2, center=<1.5068 0 0>}]
 dyn_ropeJossel_01 [{shapeType=Capsule, length=7.7099, radius=2, center=<1.5068 0 0>}]
 dyn_ropeJossel_02 [{shapeType=Capsule, length=7.7099, radius=2, center=<1.5068 0 0>}]
 dyn_whip_01 [{shapeType=Capsule, length=4.0182, radius=1.1597, center=<1.6138 -0.0001 0>, rotation=<rot 90 0 0>}]
 dyn_whip_02 [{shapeType=Capsule, length=4.8652, radius=1.0576, center=<2.7323 0.0002 0.0775>, rotation=<rot 90 0 0>}]
 dyn_whip_03 [{shapeType=Capsule, length=2.6176, radius=1.4719, center=<3.0912 0.0001 0>, rotation=<rot 90 0 0>}]
 dyn_whip_04 [{shapeType=Capsule, length=4.6372, radius=1.332, center=<3.6753 -0.0001 0>, rotation=<rot 90 0 0>}]
 dyn_whip_05 [{shapeType=Capsule, length=3.868, radius=1.332, center=<3.4899 -0.0001 0>, rotation=<rot 90 0 0>}]
 dyn_whip_06 [{shapeType=Capsule, length=1.8581, radius=1.8222, center=<2.8225 -0.0001 0>, rotation=<rot 90 0 0>}]
 dyn_whip_07 [{shapeType=Capsule, length=3.7913, radius=2.5205, center=<4.329 -0.0002 0>, rotation=<rot 90 0 0>}]
 dyn_whip_Base [{shapeType=Box, extentX=8.9172, extentY=18.0745, extentZ=1.5733, center=<1.4851 0 -0.0001>}]
 dyn_whip_MSTR [{shapeType=Capsule, length=1, radius=1}]
 pelvis [{shapeType=Capsule, length=6.3436, radius=12.725, center=<-3.8928 -0.85 0>}]
 spine_01 [{shapeType=Capsule, length=7.5997, radius=7.469}]
 spine_02 [{shapeType=Capsule, length=4.7349, radius=7.469}]
 spine_03 [{shapeType=Capsule, length=3.7668, radius=7.6939}]
 thigh_l [{shapeType=Capsule, length=27.3228, radius=7.3111, center=<-20.4206 0.0008 0>, rotation=<rot 90 0 0>}]
 thigh_r [{shapeType=Capsule, length=33.1691, radius=7.8213, center=<22.3208 -0.0005 0>, rotation=<rot 90 0 0>}]
}
constraints [
 {
  bone1Name=spine_01
  bone2Name=pelvis
  swing1LimitDegrees=45
  swing1Motion=Free
  swing2LimitDegrees=45
  swing2Motion=Free
  twistLimitDegrees=45
  twistMotion=Free
 }
 {
  bone1Name=spine_02
  bone2Name=spine_01
  swing1LimitDegrees=45
  swing1Motion=Free
  swing2LimitDegrees=45
  swing2Motion=Free
  twistLimitDegrees=45
  twistMotion=Free
 }
 {
  bone1Name=spine_03
  bone2Name=spine_02
  swing1LimitDegrees=45
  swing1Motion=Free
  swing2LimitDegrees=45
  swing2Motion=Free
  twistLimitDegrees=45
  twistMotion=Free
 }
 {bone1Name=dyn_whip_01, bone2Name=dyn_whip_MSTR, swing1LimitDegrees=45, swing2LimitDegrees=45, twistLimitDegrees=20}
 {bone1Name=dyn_whip_02, bone2Name=dyn_whip_01, swing1LimitDegrees=45, swing2LimitDegrees=45, twistLimitDegrees=20}
 {bone1Name=dyn_whip_03, bone2Name=dyn_whip_02, swing1LimitDegrees=45, swing2LimitDegrees=45, twistLimitDegrees=20}
 {bone1Name=dyn_whip_04, bone2Name=dyn_whip_03, swing1LimitDegrees=45, swing2LimitDegrees=45, twistLimitDegrees=20}
 {bone1Name=dyn_whip_05, bone2Name=dyn_whip_04, swing1LimitDegrees=45, swing2LimitDegrees=45, twistLimitDegrees=20}
 {bone1Name=dyn_whip_06, bone2Name=dyn_whip_05, swing1LimitDegrees=45, swing2LimitDegrees=45, twistLimitDegrees=20}
 {bone1Name=dyn_whip_07, bone2Name=dyn_whip_06, swing1LimitDegrees=45, swing2LimitDegrees=45, twistLimitDegrees=20}
 {
  bone1Name=thigh_l
  bone2Name=pelvis
  swing1LimitDegrees=45
  swing1Motion=Free
  swing2LimitDegrees=45
  swing2Motion=Free
  twistLimitDegrees=45
  twistMotion=Free
 }
 {
  bone1Name=thigh_r
  bone2Name=pelvis
  swing1LimitDegrees=45
  swing1Motion=Free
  swing2LimitDegrees=45
  swing2Motion=Free
  twistLimitDegrees=45
  twistMotion=Free
 }
 {bone1Name=dyn_ropeJossel_02, bone2Name=spine_01, swing1LimitDegrees=10, swing2LimitDegrees=10, twistLimitDegrees=10}
 {bone1Name=dyn_ropeJossel_01, bone2Name=pelvis, swing1LimitDegrees=10, swing2LimitDegrees=10, twistLimitDegrees=10}
 {bone1Name=dyn_hilt, bone2Name=pelvis, swing1LimitDegrees=10, swing2LimitDegrees=10, twistLimitDegrees=10}
 {bone1Name=dyn_whip_Base, bone2Name=pelvis, swing1LimitDegrees=45, swing2LimitDegrees=45, twistLimitDegrees=45}
 {
  bone1Name=calf_r
  bone2Name=thigh_r
  swing1LimitDegrees=45
  swing1Motion=Free
  swing2LimitDegrees=45
  swing2Motion=Free
  twistLimitDegrees=45
  twistMotion=Free
 }
 {
  bone1Name=dyn_whip_MSTR
  bone2Name=pelvis
  swing1LimitDegrees=45
  swing1Motion=Free
  swing2LimitDegrees=45
  swing2Motion=Free
  twistLimitDegrees=45
  twistMotion=Free
 }
]
