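# Physics asset for the whip on the F_MED_Cavern female medium body cosmetic, covering the whip and its nearby body parts.
# The whip is a seven-capsule chain running from a master bone constrained to the pelvis, alongside a box base and hilt.
# Two rope jossel bodies hang with tight limits from pelvis and lower spine.
# Pelvis, spine, thigh, extra dyn thigh and right calf capsules supply body volumes, joined by free-rotation constraints.
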PhysicsAsset /BRCosmetics/Characters/Player/Female/Medium/Bodies/F_MED_Cavern/Meshes/F_MED_Cavern_Whip_Physics
bodies {
 calf_r [{shapeType=Capsule, length=23.4408, radius=7.0201, center=<19.0238 0.0004 0>, rotation=<rot 90 0 0>}]
 dyn_hilt [{shapeType=Capsule, length=7.7099, radius=2, center=<1.5068 0 0>}]
 dyn_ropeJossel_01 [{shapeType=Capsule, length=7.7099, radius=2, center=<1.5068 0 0>}]
 dyn_ropeJossel_02 [{shapeType=Capsule, length=7.7099, radius=2, center=<1.5068 0 0>}]
 dyn_thigh_l [{shapeType=Capsule, length=27.3228, radius=7.3111, center=<-20.4206 0.0008 0>, rotation=<rot 90 0 0>}]
 dyn_thigh_r [{shapeType=Capsule, length=59.7518, radius=8.0841, center=<22.4046 -0.0005 -0.5698>, rotation=<rot 90 0 0>}]
 dyn_whip_01 [{shapeType=Capsule, length=2.5348, radius=0.6801, center=<2.1986 -0.0001 0>, rotation=<rot 90 0 0>}]
 dyn_whip_02 [{shapeType=Capsule, length=3.9492, radius=1.2508, center=<3.221 0 0>, rotation=<rot 90 0 0>}]
 dyn_whip_03 [{shapeType=Capsule, length=4.0259, radius=1.1741, center=<3.0163 0 0>, rotation=<rot 90 0 0>}]
 dyn_whip_04 [{shapeType=Capsule, length=3.6753, radius=1.5247, center=<3.0979 0 0>, rotation=<rot 90 0 0>}]
 dyn_whip_05 [{shapeType=Capsule, length=3.755, radius=1.445, center=<3.1636 0 0>, rotation=<rot 90 0 0>}]
 dyn_whip_06 [{shapeType=Capsule, length=4.7947, radius=1.5236, center=<2.3784 0 0>, rotation=<rot 90 0 0>}]
 dyn_whip_07 [{shapeType=Capsule, length=4.4944, radius=1.8175, center=<4.9504 0.0001 0>, rotation=<rot 90 0 0>}]
 dyn_whip_Base [{shapeType=Box, extentX=8.9172, extentY=18.0745, extentZ=1.5733, center=<1.4851 0 -0.0001>}]
 dyn_whip_MSTR [{shapeType=Capsule, length=1, radius=1}]
 pelvis [{shapeType=Capsule, length=7.0629, radius=13.0127, center=<-3.7623 -0.2953 0>}]
 spine_01 [{shapeType=Capsule, length=7.5997, radius=7.469}]
 spine_02 [{shapeType=Capsule, length=4.7349, radius=7.469}]
 spine_03 [{shapeType=Capsule, length=3.7668, radius=7.6939}]
 thigh_l [{shapeType=Capsule, length=27.3228, radius=7.3111, center=<-20.4206 0.0008 0>, rotation=<rot 90 0 0>}]
 thigh_r [{shapeType=Capsule, length=27.0635, radius=7.5704, center=<19.5615 -0.0004 0>, rotation=<rot 90 0 0>}]
}
constraints [
 {
  bone1Name=spine_01
  bone2Name=pelvis
  swing1LimitDegrees=45
  swing1Motion=Free
  swing2LimitDegrees=45
  swing2Motion=Free
  twistLimitDegrees=45
  twistMotion=Free
 }
 {
  bone1Name=spine_02
  bone2Name=spine_01
  swing1LimitDegrees=45
  swing1Motion=Free
  swing2LimitDegrees=45
  swing2Motion=Free
  twistLimitDegrees=45
  twistMotion=Free
 }
 {
  bone1Name=spine_03
  bone2Name=spine_02
  swing1LimitDegrees=45
  swing1Motion=Free
  swing2LimitDegrees=45
  swing2Motion=Free
  twistLimitDegrees=45
  twistMotion=Free
 }
 {bone1Name=dyn_whip_01, bone2Name=dyn_whip_MSTR, swing1LimitDegrees=60, swing2LimitDegrees=60, twistLimitDegrees=60}
 {bone1Name=dyn_whip_02, bone2Name=dyn_whip_01, swing1LimitDegrees=60, swing2LimitDegrees=60, twistLimitDegrees=60}
 {bone1Name=dyn_whip_03, bone2Name=dyn_whip_02, swing1LimitDegrees=60, swing2LimitDegrees=60, twistLimitDegrees=60}
 {bone1Name=dyn_whip_04, bone2Name=dyn_whip_03, swing1LimitDegrees=60, swing2LimitDegrees=60, twistLimitDegrees=60}
 {bone1Name=dyn_whip_05, bone2Name=dyn_whip_04, swing1LimitDegrees=60, swing2LimitDegrees=60, twistLimitDegrees=60}
 {bone1Name=dyn_whip_06, bone2Name=dyn_whip_05, swing1LimitDegrees=60, swing2LimitDegrees=60, twistLimitDegrees=60}
 {bone1Name=dyn_whip_07, bone2Name=dyn_whip_06, swing1LimitDegrees=60, swing2LimitDegrees=60, twistLimitDegrees=60}
 {
  bone1Name=thigh_l
  bone2Name=pelvis
  swing1LimitDegrees=45
  swing1Motion=Free
  swing2LimitDegrees=45
  swing2Motion=Free
  twistLimitDegrees=45
  twistMotion=Free
 }
 {
  bone1Name=thigh_r
  bone2Name=pelvis
  swing1LimitDegrees=45
  swing1Motion=Free
  swing2LimitDegrees=45
  swing2Motion=Free
  twistLimitDegrees=45
  twistMotion=Free
 }
 {bone1Name=dyn_ropeJossel_02, bone2Name=spine_01, swing1LimitDegrees=10, swing2LimitDegrees=10, twistLimitDegrees=10}
 {bone1Name=dyn_ropeJossel_01, bone2Name=pelvis, swing1LimitDegrees=10, swing2LimitDegrees=10, twistLimitDegrees=10}
 {bone1Name=dyn_hilt, bone2Name=pelvis, swing1LimitDegrees=10, swing2LimitDegrees=10, twistLimitDegrees=10}
 {bone1Name=dyn_whip_Base, bone2Name=pelvis, swing1LimitDegrees=45, swing2LimitDegrees=45, twistLimitDegrees=45}
 {
  bone1Name=dyn_thigh_l
  bone2Name=thigh_l
  swing1LimitDegrees=45
  swing1Motion=Free
  swing2LimitDegrees=45
  swing2Motion=Free
  twistLimitDegrees=45
  twistMotion=Free
 }
 {
  bone1Name=dyn_thigh_r
  bone2Name=thigh_r
  swing1LimitDegrees=45
  swing1Motion=Free
  swing2LimitDegrees=45
  swing2Motion=Free
  twistLimitDegrees=45
  twistMotion=Free
 }
 {
  bone1Name=calf_r
  bone2Name=thigh_r
  swing1LimitDegrees=45
  swing1Motion=Free
  swing2LimitDegrees=45
  swing2Motion=Free
  twistLimitDegrees=45
  twistMotion=Free
 }
 {bone1Name=dyn_whip_MSTR, bone2Name=pelvis, swing1LimitDegrees=45, swing2LimitDegrees=45, twistLimitDegrees=45}
]
