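# Character-part AnimBP for the F_MED_Cavern female body, adding deformation and secondary motion to its whip, necklace, pouch and clothing.
# The parent's pose passes through twist and sculpt Control Rigs, constraints and whip RigidBody physics into a cache, with a second whip simulation blended in when crouching.
# Accessory RigidBody, AnimDynamics and corrective bone offsets follow.
# Updates set state-based gravity and a computed sim-space rotation driving the physics base bone.
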AnimBlueprint /BRCosmetics/Characters/Player/Female/Medium/Bodies/F_MED_Cavern/Meshes/F_MED_Cavern_AnimBP
target_skeleton=&/F_MED_Cavern_Skeleton
graphs [
 AnimGraph {nodes [
   "Transform (Modify) Bone - Bone: dyn_whip_MSTR"!(ComponentPose, bAlphaBoolEnabled:bool=Get bIsMoving2D(), Translation="0.000,-1,7", Rotation="0.000,0.000,-15", Scale="1,1,1") -> (Pose) {
    alphaInputType=Bool
    rotationMode=BMM_Additive
    scaleMode=BMM_Additive
    translationMode=BMM_Additive
    alphaBoolBlend {blendInTime=0.2, blendOption=Linear, blendOutTime=0.1}
    boneToModify {boneName=dyn_whip_MSTR}
    rotation {roll=-15}
    translation {y=-1, z=7}
   }
   n1 = Control Rig(Source=Use cached pose 'PreSim'()) {
    alphaBoolBlend {blendOption=Linear}
    controlRigAssetReference {blueprintRigClass=&/F_MED_Cavern_Deform_CtrlRig_twists.F_MED_Cavern_Deform_CtrlRig_twists_C}
   }
   n2 = Control Rig(Source=@n1) {
    alphaBoolBlend {blendOption=Linear}
    controlRigAssetReference {blueprintRigClass=&/F_MED_Cavern_Deform_CtrlRig_sculpters.F_MED_Cavern_Deform_CtrlRig_sculpters_C}
   }
   n3 = "Constraint - dyn_whip_MSTR (PP)"!(ComponentPose=Local To Component(LocalPose=@n2), Alpha=1, ConstraintWeights_0=0.6, ConstraintWeights_1:float=0.4) -> (Pose) {
    alphaBoolBlend {blendOption=Linear}
    boneToModify {boneName=dyn_whip_MSTR}
    constraintSetup [{targetBone {boneName=thigh_twist_01_r}}, {targetBone {boneName=pelvis}}]
    constraintWeights [1,1]
   }
   n4 = "Constraint - dyn_whip_01 (RR)"!(ComponentPose=@n3, Alpha=1, ConstraintWeights_0=1, ConstraintWeights_1:float=0) -> (Pose) {
    alphaBoolBlend {blendOption=Linear}
    boneToModify {boneName=dyn_whip_01}
    constraintSetup [{transformType=Rotation, targetBone {boneName=thigh_twist_01_r}}, {transformType=Rotation, targetBone {boneName=pelvis}}]
    constraintWeights [1,1]
   }
   n5 = "Transform (Modify) Bone - Bone: dyn_whip_MSTR"!(ComponentPose=@n4, Alpha:float=Get LegRot_R() -> (LegRot_R:float), Translation="0,-3,1", Rotation="0.000,0.000,-15", Scale="1,1,1") -> (Pose) {
    rotationMode=BMM_Additive
    rotationSpace=BCS_BoneSpace
    scaleMode=BMM_Additive
    translationMode=BMM_Additive
    translationSpace=BCS_BoneSpace
    alphaBoolBlend {blendInTime=0.2, blendOption=Linear, blendOutTime=0.1}
    boneToModify {boneName=dyn_whip_MSTR}
    rotation {roll=-15}
    translation {y=-3, z=1}
   }
   n6 = RigidBody(ComponentPose=@n5, ExternalForce=Get GravityOverride()) {
    bClampLinearTranslationLimitToRefPose=true
    bOverrideWorldGravity=true
    overridePhysicsAsset=&/F_MED_Cavern_WhipCrouch_Physics
    simulationSpace=BaseBoneSpace
    alphaBoolBlend {blendOption=Linear}
    baseBoneRef {boneName=dyn_root_simspace}
    componentAppliedLinearAccClamp=<2500 2500 2500>
    overrideWorldGravity {z=-980}
   }
   n7 = "Transform (Modify) Bone - Bone: dyn_whip_Base"!(ComponentPose=@n6, bAlphaBoolEnabled:bool=True, Translation="-3,1,0", Rotation="0,0,0", Scale="1,1,1") -> (Pose) {
    alphaInputType=Bool
    rotationMode=BMM_Additive
    rotationSpace=BCS_BoneSpace
    scaleMode=BMM_Additive
    scaleSpace=BCS_BoneSpace
    translationMode=BMM_Additive
    translationSpace=BCS_BoneSpace
    alphaBoolBlend {blendOption=Linear}
    boneToModify {boneName=dyn_whip_Base}
    translation=<-3 1>
   }
   n8 = Component To Local(ComponentPose=@n7)
   n9 = Use cached pose 'PreSim'()
   n10 = Get bIsCrouching()
   n11 = "Constraint - dyn_necklace_MSTR (P)"(
    ComponentPose=Local To Component(LocalPose=Blend Poses by bool(BlendPose_0=@n8, BlendPose_1=@n9, BlendTime_0=0.3, BlendTime_1=0.3, bActiveValue=@n10) {blendType=Linear})
    Alpha=1
    ConstraintWeights_0=1
   ) {
    alphaBoolBlend {blendOption=Linear}
    boneToModify {boneName=dyn_necklace_MSTR}
    constraintSetup [{targetBone {boneName=spine_05}}]
    constraintWeights [1]
   }
   n12 = "Constraint - dyn_pouchZipper_MSTR (P)"(ComponentPose=@n11, Alpha=1, ConstraintWeights_0=1) {
    alphaBoolBlend {blendOption=Linear}
    boneToModify {boneName=dyn_pouchZipper_MSTR}
    constraintSetup [{targetBone {boneName=thigh_twist_01_l}}]
    constraintWeights [1]
   }
   n13 = "Constraint - dyn_pouchTab_MSTR (P)"(ComponentPose=@n12, Alpha=1, ConstraintWeights_0=1) {
    alphaBoolBlend {blendOption=Linear}
    boneToModify {boneName=dyn_pouchTab_MSTR}
    constraintSetup [{targetBone {boneName=thigh_twist_01_l}}]
    constraintWeights [1]
   }
   n14 = RigidBody!(ComponentPose=@n13, Alpha, ExternalForce=Get GravityOverride(), bTransferBoneVelocities:bool=False) -> (Pose) {
    bClampLinearTranslationLimitToRefPose=true
    bOverrideWorldGravity=true
    overridePhysicsAsset=&/F_MED_Cavern_Accessories_Physics
    simulationSpace=BaseBoneSpace
    alphaBoolBlend {blendOption=Linear}
    baseBoneRef {boneName=dyn_root_simspace}
    componentAppliedLinearAccClamp=<500 500 500>
    componentLinearAccScale=<0.4 0.4 0.4>
    overrideWorldGravity {z=-980}
   }
   n15 = Anim Dynamics - Bone: dyn_necklace(ComponentPose=@n14, Alpha=1, ExternalForce="0,0,0") {
    angularDampingOverride=0.9999
    bGravityOverrideInSimSpace=true
    bOverrideAngularDamping=true
    simulationSpace=RootRelative
    alphaBoolBlend {blendOption=Linear}
    boundBone {boneName=dyn_necklace}
    physicsBodyDefinitions [{
      sphereCollisionRadius=0
      boundBone {boneName=dyn_necklace}
      boxExtents=<3 10 10>
      constraintSetup {angularLimitsMax=<2 3 45>, angularLimitsMin=<-2 -75 -45>, angularTarget=<0 0 0>}
      localJointOffset=<1.5 -0.05 0.1>
     }]
   }
   n16 = "Constraint - dyn_pouchZipper_MSTR (P)"!(ComponentPose=@n15, Alpha=1, ConstraintWeights_0=1, ConstraintWeights_1:float=0) -> (Pose) {
    alphaBoolBlend {blendOption=Linear}
    boneToModify {boneName=dyn_pouchZipper_MSTR}
    constraintSetup [{targetBone {boneName=thigh_twist_01_l}}]
    constraintWeights [1,0]
   }
   n17 = Anim Dynamics - Bone: dyn_pouchZipper(ComponentPose=@n16, Alpha=1, ExternalForce="0,0,0") {
    angularDampingOverride=0.8
    bGravityOverrideInSimSpace=true
    bOverrideAngularDamping=true
    simulationSpace=RootRelative
    alphaBoolBlend {blendOption=Linear}
    boundBone {boneName=dyn_pouchZipper}
    componentLinearAccScale=<1 1 1>
    componentLinearVelScale=<1 1 1>
    physicsBodyDefinitions [{
      sphereCollisionRadius=0
      boundBone {boneName=dyn_pouchZipper}
      boxExtents=<3 10 10>
      constraintSetup {angularLimitsMax=<45 0 45>, angularLimitsMin=<-45 -110 -45>, angularTarget=<0 0 0>}
      localJointOffset=<-1 -0.05 0.1>
     }]
   }
   n18 = "Transform (Modify) Bone - Bone: dyn_clavicle_l"(ComponentPose=@n17, Alpha=Get LookUp_L() -> (LookUp_L:float), Translation="0,0,1", Rotation="0,15,0", Scale="1,1.5,1") {
    rotationMode=BMM_Additive
    rotationSpace=BCS_BoneSpace
    scaleMode=BMM_Additive
    scaleSpace=BCS_BoneSpace
    translationMode=BMM_Additive
    translationSpace=BCS_BoneSpace
    alphaBoolBlend {blendOption=Linear}
    boneToModify {boneName=dyn_clavicle_l}
    rotation {yaw=15}
    scale {y=1.5}
    translation {z=1}
   }
   n19 = "Transform (Modify) Bone - Bone: dyn_clavicle_r"(ComponentPose=@n18, Alpha=Get LookUp_R() -> (LookUp_R:float), Translation="0,0,0", Rotation="0,25,0", Scale="1,1.5,1") {
    rotationMode=BMM_Additive
    rotationSpace=BCS_BoneSpace
    scaleMode=BMM_Additive
    scaleSpace=BCS_BoneSpace
    translationMode=BMM_Additive
    translationSpace=BCS_BoneSpace
    alphaBoolBlend {blendOption=Linear}
    boneToModify {boneName=dyn_clavicle_r}
    rotation {yaw=25}
    scale {y=1.5}
   }
   n20 = "Transform (Modify) Bone - Bone: dyn_hoodFix_L"(ComponentPose=@n19, Alpha=1, Translation="-0.811,-0.000,1.136", Rotation="0,0,0", Scale="1,1.5,1") {
    rotationMode=BMM_Additive
    rotationSpace=BCS_BoneSpace
    scaleMode=BMM_Additive
    scaleSpace=BCS_BoneSpace
    translationMode=BMM_Additive
    translationSpace=BCS_BoneSpace
    alphaBoolBlend {blendOption=Linear}
    boneToModify {boneName=dyn_hoodFix_L}
    scale {y=1.5}
    translation=<-0.811 0 1.1361>
   }
   n21 = "Transform (Modify) Bone - Bone: dyn_hoodFix_R"(ComponentPose=@n20, Alpha=1, Translation="-0.000,0.059,-2.157", Rotation="0.000,-0.000,-18.523", Scale="1,1.5,1") {
    rotationMode=BMM_Additive
    rotationSpace=BCS_BoneSpace
    scaleMode=BMM_Additive
    scaleSpace=BCS_BoneSpace
    translationMode=BMM_Additive
    translationSpace=BCS_BoneSpace
    alphaBoolBlend {blendOption=Linear}
    boneToModify {boneName=dyn_hoodFix_R}
    rotation {roll=-18.523}
    scale {y=1.5}
    translation {y=0.059, z=-2.157}
   }
   n22 = "Transform (Modify) Bone - Bone: dyn_root_slope"!(
    ComponentPose=@n21
    bAlphaBoolEnabled:bool=NOT Boolean(A=Get bCanPlayCustomAnimations() -> (bCanPlayCustomAnimations:bool))
    Translation=Get SlopeRootTrans() -> (SlopeRootTrans:Vector)
    Rotation=Get SlopeFootRot() -> (SlopeFootRot:Rotator)
    Scale="1,1,1"
   ) -> (Pose) {
    alphaInputType=Bool
    rotationMode=BMM_Replace
    rotationSpace=BCS_WorldSpace
    translationMode=BMM_Replace
    translationSpace=BCS_WorldSpace
    alphaBoolBlend {blendOption=Linear}
    boneToModify {boneName=dyn_root_slope}
   }
   n23 = "Transform (Modify) Bone - Bone: dyn_root_simspace"(ComponentPose=@n22, Alpha=1, Translation="0,0,0", Rotation=Get SimSpaceRot() -> (SimSpaceRot:Rotator), Scale="1,1,1") {
    rotationMode=BMM_Replace
    rotationSpace=BCS_WorldSpace
    translationSpace=BCS_BoneSpace
    alphaBoolBlend {blendOption=Linear}
    alphaScaleBiasClamp {bInterpResult=true, interpSpeedDecreasing=2, interpSpeedIncreasing=2}
    boneToModify {boneName=dyn_root_simspace}
   }
   n24 = "Transform (Modify) Bone - Bone: dyn_left_shirt"(ComponentPose=@n23, Alpha=1, Translation="3,-2,0", Rotation="0,0,0", Scale="1,1,1") {
    rotationMode=BMM_Additive
    rotationSpace=BCS_WorldSpace
    scaleMode=BMM_Additive
    translationMode=BMM_Additive
    translationSpace=BCS_BoneSpace
    alphaBoolBlend {blendOption=Linear}
    alphaScaleBiasClamp {bInterpResult=true, interpSpeedDecreasing=2, interpSpeedIncreasing=2}
    boneToModify {boneName=dyn_left_shirt}
    translation=<3 -2>
   }
   n25 = Anim Dynamics - Bone: dyn_back_shirt(ComponentPose=@n24, Alpha=1, ExternalForce="0,0,0") {
    angularDampingOverride=0.9999
    angularSpringConstant=75
    bAngularSpring=true
    bGravityOverrideInSimSpace=true
    bOverrideAngularDamping=true
    simulationSpace=RootRelative
    alphaBoolBlend {blendOption=Linear}
    boundBone {boneName=dyn_back_shirt}
    componentLinearAccScale=<1 1 1>
    componentLinearVelScale=<1 1 1>
    physicsBodyDefinitions [{
      sphereCollisionRadius=0
      boundBone {boneName=dyn_back_shirt}
      boxExtents=<20 20 10>
      constraintSetup {angularLimitsMax=<5 -5 5>, angularLimitsMin=<-5 -25 -5>, angularTarget=<0 0 0>}
      localJointOffset=<2 0 0>
     }]
   }
   Output Pose(Result=Component To Local(ComponentPose=@n25)) {layerGroup=None}
   n26 = Control Rig(Source=Copy Pose From Mesh() {bUseAttachedParent=true}) {
    alphaBoolBlend {blendOption=Linear}
    controlRigAssetReference {blueprintRigClass=&F_MED_Cavern_Deform_CtrlRig_twists_C}
   }
   n27 = Control Rig(Source=@n26) {
    alphaBoolBlend {blendOption=Linear}
    controlRigAssetReference {blueprintRigClass=&F_MED_Cavern_Deform_CtrlRig_sculpters_C}
   }
   n28 = "Constraint - dyn_whip_MSTR (PPPPPPPPPPPPPPP)"!(
    ComponentPose=Local To Component(LocalPose=@n27)
    Alpha=1
    ConstraintWeights_0=0.0031
    ConstraintWeights_1:float=0.0087
    ConstraintWeights_2:float=0.0028
    ConstraintWeights_3:float=0
    ConstraintWeights_4:float=0.3396
    ConstraintWeights_5:float=0.0001
    ConstraintWeights_6:float=0.0064
    ConstraintWeights_7:float=0.01
    ConstraintWeights_8:float=0.0001
    ConstraintWeights_9:float=0.0015
    ConstraintWeights_10:float=0
    ConstraintWeights_11:float=0.4032
    ConstraintWeights_12:float=0.0069
    ConstraintWeights_13:float=0
    ConstraintWeights_14:float=0.2175
   ) -> (Pose) {
    alphaBoolBlend {blendOption=Linear}
    boneToModify {boneName=dyn_whip_MSTR}
    constraintSetup [
     {targetBone {boneName=deform_thigh_fr_l}}
     {targetBone {boneName=deform_thigh_bk_r}}
     {targetBone {boneName=thigh_r}}
     {targetBone {boneName=dyn_whip_Base}}
     {targetBone {boneName=thigh_twist_01_r}}
     {targetBone {boneName=dyn_l_butt_helper}}
     {targetBone {boneName=spine_01}}
     {targetBone {boneName=spine_02}}
     {targetBone {boneName=spine_03}}
     {targetBone {boneName=spine_04}}
     {targetBone {boneName=thigh_l}}
     {targetBone {boneName=pelvis}}
     {targetBone {boneName=thigh_twist_01_l}}
     {targetBone {boneName=deform_thigh_bk_l}}
     {targetBone {boneName=deform_thigh_fr_r}}
    ]
    constraintWeights [1,1,1,1,1,1,1,1,1,1,1,1,1,1,1]
   }
   n29 = "Transform (Modify) Bone - Bone: dyn_whip_Base"!(ComponentPose=@n28, Alpha:float=1, Translation="-2,0,0", Rotation="0,0,0", Scale="1,1,1") -> (Pose) {
    rotationMode=BMM_Additive
    rotationSpace=BCS_BoneSpace
    scaleMode=BMM_Additive
    scaleSpace=BCS_BoneSpace
    translationMode=BMM_Additive
    translationSpace=BCS_BoneSpace
    alphaBoolBlend {blendOption=Linear}
    boneToModify {boneName=dyn_whip_Base}
    translation {x=-2}
   }
   n30 = "Constraint - dyn_whip_01 (RR)"(ComponentPose=@n29, Alpha=1, ConstraintWeights_0=1, ConstraintWeights_1=0) {
    alphaBoolBlend {blendOption=Linear}
    boneToModify {boneName=dyn_whip_01}
    constraintSetup [{transformType=Rotation, targetBone {boneName=thigh_twist_01_r}}, {transformType=Rotation, targetBone {boneName=pelvis}}]
    constraintWeights [1,1]
   }
   n31 = RigidBody!(ComponentPose=@n30, Alpha, OverrideWorldGravity:Vector=Get GravityOverride(), ExternalForce) -> (Pose) {
    bClampLinearTranslationLimitToRefPose=true
    bOverrideWorldGravity=true
    overridePhysicsAsset=&/F_MED_Cavern_Whip_Physics
    simulationSpace=BaseBoneSpace
    alphaBoolBlend {blendOption=Linear}
    baseBoneRef {boneName=dyn_root_simspace}
    componentAppliedLinearAccClamp=<1250 1250 1250>
    componentLinearAccScale {z=0.35}
    componentLinearVelScale {z=0.35}
    overrideWorldGravity {z=-980}
    simSpaceSettings {
     maxAngularAcceleration=1000
     maxAngularVelocity=3
     maxLinearAcceleration=700
     maxLinearVelocity=500
     velocityScaleZ=0
     worldAlpha=1
     externalLinearDragV=<1 1>
    }
   }
   Save cached pose 'PreSim'(Pose=Component To Local(ComponentPose=@n31)) {cacheName=PreSim}
  ]}
 EventGraph {nodes [
   Event BlueprintUpdateAnimation()
   > n1 = Sequence()
   > n2 = Gravity Override State Function DEPRECATED(CharacterPartAnimInstance=Self-Reference(), GameTimeInSeconds=GetGameTimeInSeconds(), GravityOverrideParams=Get Gravity Override Params())
   > Set GravityOverride(GravityOverride=@n2)
   n3 = ComplexSimSpaceFunction DEPRECATED!(
    execute=@n1.then_1
    CharacterPartAnimInstance:Custom Character Part Anim Instance Object Reference=Self-Reference()
    InterpSpeedMax:double=9
    PawnSpeedMax:double=20
    RInterpErrorTolerance:double=1
    DeltaTime:double=0
    DebugDraw:bool=false
    sim_space_rotation_in:Rotator=Get SimSpaceRot()
    inComponentSpaceIn:bool=Get InComponentSpace() -> (InComponentSpace:bool)
    inComponentSpaceLastFrameIn:bool=Get InComponentSpaceLastFrame() -> (InComponentSpaceLastFrame:bool)
    lockedOnSimSpaceIn:bool=Get LockedOnSimSpace() -> (LockedOnSimSpace:bool)
    InterpSpeedIn:double=Get InterpSpeed() -> (InterpSpeed:double)
   ) -> (
    then
    sim_space_rotation:Rotator
    inComponentSpaceOut:bool
    inComponentSpaceLastFrameOut:bool
    lockedOnSimSpaceOut:bool
    InterpSpeedOut:double
   )
   > Set SimSpaceRot(SimSpaceRot:Rotator=@n3) -> (Output_Get:Rotator)
   > Set InComponentSpace(InComponentSpace:bool=@n3.inComponentSpaceOut) -> (Output_Get:bool)
   > Set InComponentSpaceLastFrame(InComponentSpaceLastFrame:bool=@n3.inComponentSpaceLastFrameOut) -> (Output_Get:bool)
   > Set LockedOnSimSpace(LockedOnSimSpace:bool=@n3.lockedOnSimSpaceOut) -> (Output_Get:bool)
   > Set InterpSpeed(InterpSpeed:double=@n3.InterpSpeedOut) -> (Output_Get:double)
   Branch(Condition=Get bIsInFrontEnd())
   > n4 = Gravity Override State Function DEPRECATED(CharacterPartAnimInstance=Self-Reference(), GameTimeInSeconds=GetGameTimeInSeconds(), GravityOverrideParams=Get Gravity Override Params())
   > Set GravityOverride(GravityOverride=@n4)
  ]}
]
variables [
 Gravity Override Params {
  type=GravityOverrideParamsStruct
  default {
   emoteOrMelee {gravityOverride=<0 0 -1100>}
   front_End {gravityOverride=<0 0 -1100>}
   onGround_Moving {gravityOverride=<0 0 -1100>}
   onGround_Standing {gravityOverride=<0 0 -1100>}
   skydive_DivingDown {gravityOverride=<0 0 0>, windAmplitude=<100 -250 200>, windFrequency=<1.5 0.5 2>, windOffset=<0 -600 400>}
   skydive_FloatingDown {gravityOverride=<0 0 250>, windAmplitude=<100 -250 200>, windFrequency=<1.5 0.5 2>}
   skydive_GoingUp {gravityOverride=<0 0 -100>, windAmplitude=<100 -250 200>, windFrequency=<1.5 0.5 2>, windOffset=<0 -600 400>}
   skydive_Parachuting {gravityOverride=<0 0 -980>, windAmplitude=<100 -250 200>, windFrequency=<1.5 0.5 2>, windOffset=<0 -600 400>}
   swimming {gravityOverride=<0 0 0>}
  }
 }
 GravityOverride {type=Vector}
 InComponentSpace {type=boolean}
 InComponentSpaceLastFrame {type=boolean}
 InterpSpeed {type=number}
 LockedOnSimSpace {type=boolean}
]
